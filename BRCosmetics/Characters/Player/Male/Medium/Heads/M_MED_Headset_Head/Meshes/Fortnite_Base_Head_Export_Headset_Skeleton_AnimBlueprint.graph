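# Post-adjusts the face pose of the Headset cosmetic character's head when its Mat2 material variant is active, apparently to fit a mask.
# Each update reads the pawn's cosmetic variants into maskIsOn and is Fantastical, and measures jaw rotation.
# maskIsOn gates chained Modify Bone nodes that override jaw rotation and offset lower lip and cheeks.
# The pose-asset blend is unconnected; is Fantastical and the jaw measurement never reach the output pose.
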AnimBlueprint /BRCosmetics/Characters/Player/Male/Medium/Heads/M_MED_Headset_Head/Meshes/Fortnite_Base_Head_Export_Headset_Skeleton_AnimBlueprint
target_skeleton=/Game/Characters/Player/Common/Fortnite_Base_Head/Fortnite_Base_Head_Export_Skeleton
graphs [
 AnimGraph {nodes [
   Blend Poses by bool(
    BlendPose_0=M_MED_Headset_Head_PoseAsset : Pose_2!(PoseAsset:Pose Asset Object Reference=&/M_MED_Headset_Head_PoseAsset) -> (Pose:Pose Link Structure) {is=title AnimGraphNode_PoseByName, poseName=Pose_2, poseWeight=0.25}
    BlendPose_1="Pose (None)"!(PoseAsset:Pose Asset Object Reference) -> (Pose:Pose Link Structure) {is=title AnimGraphNode_PoseByName, poseName=Pose_2, poseWeight=0.25}
    bActiveValue=Get maskIsOn() -> (maskIsOn:bool)
   ) {blendType=Linear}
   Component To Local(ComponentPose=Local To Component())
   "Transform (Modify) Bone - Bone: C_jaw"(Alpha=1, Translation="0,0,0", Rotation=Get head rot() -> (head rot:Rotator), Scale="1,1,1") {rotationMode=BMM_Additive, rotationSpace=BCS_BoneSpace, alphaBoolBlend {blendOption=Linear}, boneToModify {boneName=C_jaw}}
   n1 = Get maskIsOn()
   n2 = "Transform (Modify) Bone - Bone: C_jaw"!(ComponentPose=Local To Component(LocalPose=Input Pose()), bAlphaBoolEnabled:bool=@n1, Translation="0,0,0", Rotation="0,0,0", Scale="1,1,1") -> (Pose) {
    alphaInputType=Bool
    rotationMode=BMM_Replace
    rotationSpace=BCS_ParentBoneSpace
    translationMode=BMM_Additive
    translationSpace=BCS_ParentBoneSpace
    alphaBoolBlend {blendOption=Linear}
    boneToModify {boneName=C_jaw}
   }
   n3 = "Transform (Modify) Bone - Bone: C_lip_lower_mid"!(ComponentPose=@n2, bAlphaBoolEnabled:bool=@n1, Translation="0,-2,5", Rotation="0,0,0", Scale="1,1,1") -> (Pose) {
    alphaInputType=Bool
    rotationMode=BMM_Replace
    rotationSpace=BCS_ParentBoneSpace
    translationMode=BMM_Additive
    translationSpace=BCS_ParentBoneSpace
    alphaBoolBlend {blendOption=Linear}
    boneToModify {boneName=C_lip_lower_mid}
    translation {y=-2, z=5}
   }
   n4 = "Transform (Modify) Bone - Bone: R_cheek_inner"!(ComponentPose=@n3, bAlphaBoolEnabled:bool=@n1, Translation="0,-1,0", Rotation="0,0,0", Scale="0.5,0.5,1") -> (Pose) {
    alphaInputType=Bool
    rotationMode=BMM_Replace
    rotationSpace=BCS_ParentBoneSpace
    scaleMode=BMM_Additive
    scaleSpace=BCS_BoneSpace
    translationMode=BMM_Additive
    translationSpace=BCS_BoneSpace
    alphaBoolBlend {blendOption=Linear}
    boneToModify {boneName=R_cheek_inner}
    scale=<0.5 0.5>
    translation {y=-1}
   }
   n5 = "Transform (Modify) Bone - Bone: L_cheek_inner"!(ComponentPose=@n4, bAlphaBoolEnabled:bool=@n1, Translation="-1,0,0", Rotation="0,0,0", Scale="0.5,0.5,1") -> (Pose) {
    alphaInputType=Bool
    rotationMode=BMM_Replace
    rotationSpace=BCS_ParentBoneSpace
    scaleMode=BMM_Additive
    scaleSpace=BCS_BoneSpace
    translationMode=BMM_Additive
    translationSpace=BCS_BoneSpace
    alphaBoolBlend {blendOption=Linear}
    boneToModify {boneName=L_cheek_inner}
    scale=<0.5 0.5>
    translation {x=-1}
   }
   Output Pose(Result=Component To Local(ComponentPose=@n5)) {layerGroup=None}
   Save cached pose 'SavedPose26247'() {cacheName=SavedPose26247}
  ]}
 EventGraph {nodes [
   Event BlueprintUpdateAnimation()
   > n1 = Sequence!(execute) -> (then_0, then_1, then_2:Exec)
   > Set Raw Output(Raw Output:float=GetAlphaFromReferenceTransformDelta(
     SkeletalMeshComponent=GetPartSkeletalMeshComponent()
     BoneName=C_jaw
     BaseName=head
     DeltaComponentType=EulerX
     InputRange=(X=-30,Y=30)
     TwistAxis="1,0,0"
    ).RawOutput) -> (Output_Get:float)
   > Set head rot(head rot:Rotator="0, 0, 0") -> (Output_Get:Rotator)
   n2 = Cast To FortPlayerPawn(execute=@n1.then_1, Object=GetOwningActor())
   > n3 = GetVariantChannelActiveIndexForItemActor!(
    execute
    Item:Athena Cosmetic Item Definition Object Reference=/BRCosmetics/Athena/Items/Cosmetics/Characters/Character_Headset
    Actor:Actor Object Reference=@n2
    Channel:Gameplay Tag Structure=(TagName="Cosmetics.Variant.Channel.Material")
    VariantList:Array of Gameplay Tag Structures=Make Array("[0]"=(TagName="Cosmetics.Variant.Property.Mat2"))
    bErrorIfChannelMissing:bool=false
    bErrorIfTagNotInArray:bool=false
   ) -> (then, Index:int=0, bIndexIsValid:bool, bFoundChannel:bool)
   > n4 = Branch(Condition=@n3.bIndexIsValid)
   > Set maskIsOn(maskIsOn:bool=true) -> (Output_Get:bool)
   Set maskIsOn(execute=@n4.else, maskIsOn=false)
   n5 = Cast To FortPlayerPawn(execute=@n1.then_2, Object=GetOwningActor())
   > n6 = GetVariantChannelActiveIndexForItemActor(
    Item=&Character_Headset
    Actor=@n5
    Channel=(TagName="Cosmetics.Variant.Channel.Parts")
    VariantList=Make Array("[0]"=(TagName="Cosmetics.Variant.Property.Stage1"))
    bErrorIfChannelMissing=false
    bErrorIfTagNotInArray=false
   ) -> (Index=0)
   > n7 = Branch(Condition=AND Boolean(A="Equal (Integer)"(A=@n6), B=@n6.bIndexIsValid))
   > Set is Fantastical(is Fantastical:bool=false) -> (Output_Get:bool)
   Set is Fantastical(execute=@n7.else, is Fantastical=true)
   n8 = Set jaw Translate(jaw Translate:Vector="-0.0019,3.2542,0.4325") -> (Output_Get:Vector)
   > Set jaw Translate(jaw Translate=vector * vector(A=@n8, B:Vector="-1,-1,-1") -> (ReturnValue=~))
   BreakTransform()
  ]}
]
variables [
 As Pose Asset {type=PoseAsset, default=null}
 head rot {type=Rotator}
 Head Rot Percentage {type=number}
 Index {type=integer}
 Index_0 {type=integer}
 is Fantastical {type=boolean}
 jaw Translate {type=Vector}
 maskIsOn {type=boolean}
 Raw Output {type=number}
]
